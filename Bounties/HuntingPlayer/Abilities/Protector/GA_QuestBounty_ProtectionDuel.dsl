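; Gameplay ability for the protector side of a player bounty, where the player guards a squadmate against the tracked hunter.
; On server activation it binds both targets' deaths and quest expiry, starts a timed urgent quest and schedules a one-minute warning.
; A squad kill of the hunter completes it; other deaths fail it if poaching is allowed, while expiry shows success.
; Outcomes drive HUD messages, cleanup and a DoOnce quest end, with custom verbs for kill and fail.
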(event Ability|EventActivateAbility
  ;[0 0 0] Activate
  (if (Networking|IsServer)
    (bind _bountytarget (CallFunction|GetBountyTarget))
    (Utilities|IsValid _bountytarget
      (:Is%20Valid
        (bind _returnvalue (Utilities|Time|SetTimerbyEvent (EventDispatchers|CreateEvent :!event Custom|OneMinuteLeft) (- (Math|Float|SelectFloat (Fortnite|GetNumericCVar "Fort.Bounty.TotalQuestTime") (ScalableFloat|GetValueatLevel (Variables|Default|GetTotalQuestTime)) (and (> (Fortnite|GetNumericCVar "Fort.Bounty.TotalQuestTime") 0.0) (Editor|InEditororPIE))) 59.0)))
        (Variables|Default|SetOneMinuteLeftTimerHandle _returnvalue)
        (EventDispatchers|AssignOnUrgentQuestTimeExpired (Class|FortPlayerControllerAthena|GetTransientQuestsComponent (Utilities|Casting|CastToFortPlayerControllerAthena (Fortnite|GetFortPlayerControllerfromActor (Ability|GetAvatarActorfromActorInfo)))) (EventDispatchers|CreateEvent :!event Custom|OnUrgentQuestTimeExpired_Event))
        (EventDispatchers|AssignOnPawnDiedDelegate (CallFunction|GetTrackedProtectorBountyTarget) (EventDispatchers|CreateEvent :!event Custom|OnProtectionTargetDied_Bind))
        (bind _bountytarget_2 (CallFunction|GetBountyTarget))
        (EventDispatchers|AssignOnPawnDiedDelegate _bountytarget_2 (EventDispatchers|CreateEvent :!event Custom|OnHuntTargetDied_Bind))
        (Utilities|FlowControl|DelayUntilNextFrame)
        (CallFunction|PlaceholderPostAbilityInitialization)))))

(event Custom|Placeholder_PostAbilityInitialization
  ;[0 0 0] OnPostAbilityInitialization
  (Ability|HasAuthority
    (:True
      (CallFunction|QuestStarted)
      (bind _asfort_urgent_quest_item_definition (Utilities|Casting|CastToFortUrgentQuestItemDefinition (Variables|FortTransientQuestGameplayAbility|GetQuestItemDefinition)))
      (bind _structout (Utilities|Struct|SetmembersinUrgentQuestData (Class|FortUrgentQuestItemDefinition|GetUrgentQuestData _asfort_urgent_quest_item_definition) (Math|Float|SelectFloat (Fortnite|GetNumericCVar "Fort.Bounty.TotalQuestTime") (ScalableFloat|GetValueatLevel (Variables|Default|GetTotalQuestTime)) (and (Editor|InEditororPIE) (> (Fortnite|GetNumericCVar "Fort.Bounty.TotalQuestTime") 0.0))) (Utilities|Text|FormatText (.EventSubtitle (Utilities|Struct|BreakUrgentQuestData (Class|FortUrgentQuestItemDefinition|GetUrgentQuestData _asfort_urgent_quest_item_definition))) (Class|FortPlayerControllerGameplay|GetSettingsFilteredPlayerName (Utilities|Casting|CastToFortPlayerControllerGameplay (Fortnite|GetFortPlayerControllerfromActor (Ability|GetAvatarActorfromActorInfo))) (CallFunction|GetTrackedProtectorBountyTarget))) (Utilities|Text|FormatText (.EventSubtitle (Utilities|Struct|BreakUrgentQuestData (Class|FortUrgentQuestItemDefinition|GetUrgentQuestData _asfort_urgent_quest_item_definition))) (Class|FortPlayerControllerGameplay|GetSettingsFilteredPlayerName (Utilities|Casting|CastToFortPlayerControllerGameplay (Fortnite|GetFortPlayerControllerfromActor (Ability|GetAvatarActorfromActorInfo))) (CallFunction|GetTrackedProtectorBountyTarget)))))
      (CallFunction|StartUrgentQuestEvent _structout))))

(event Custom|OnUrgentQuestTimeExpired_Event
  (params (UrgentQuestData :type Urgent%20Quest%20Data%20Structure%20%28by%20ref%29))
  ;[0 0 0] OnUrgentQuestTimeExpired_Event
  (bind _asfort_urgent_quest_item_definition (Utilities|Casting|CastToFortUrgentQuestItemDefinition (Variables|FortTransientQuestGameplayAbility|GetQuestItemDefinition)))
  (if (GameplayTags|Equal(GameplayTag) (Utilities|Struct|BreakUrgentQuestData (Class|FortUrgentQuestItemDefinition|GetUrgentQuestData _asfort_urgent_quest_item_definition)) (Utilities|Struct|BreakUrgentQuestData UrgentQuestData))
    (CallFunction|DisplayMessageSuccess "NSLOCTEXT(\"[A31A63EF45F22A15B2D0C599C177E596]\", \"FF3BE22041CFDFC7FF11528E11B9577E\", \"Bounty Completed\")" "NSLOCTEXT(\"[A31A63EF45F22A15B2D0C599C177E596]\", \"7DE21C7D4981F476B8591EB571575F97\", \"Protected Squadmate\")")
    (CallFunction|Cleanup)))

(event EventOnPostAbilityInitialization
  ;[0 0 0] OnPostAbilityInitialization
  (Ability|HasAuthority
    (:True
      (CallFunction|QuestStarted)
      (bind _asfort_urgent_quest_item_definition (Utilities|Casting|CastToFortUrgentQuestItemDefinition (Variables|FortTransientQuestGameplayAbility|GetQuestItemDefinition)))
      (bind _structout (Utilities|Struct|SetmembersinUrgentQuestData (Class|FortUrgentQuestItemDefinition|GetUrgentQuestData _asfort_urgent_quest_item_definition) (Math|Float|SelectFloat (Fortnite|GetNumericCVar "Fort.Bounty.TotalQuestTime") (ScalableFloat|GetValueatLevel (Variables|Default|GetTotalQuestTime)) (and (Editor|InEditororPIE) (> (Fortnite|GetNumericCVar "Fort.Bounty.TotalQuestTime") 0.0))) (Utilities|Text|FormatText (.EventSubtitle (Utilities|Struct|BreakUrgentQuestData (Class|FortUrgentQuestItemDefinition|GetUrgentQuestData _asfort_urgent_quest_item_definition))) (Class|FortPlayerControllerGameplay|GetSettingsFilteredPlayerName (Utilities|Casting|CastToFortPlayerControllerGameplay (Fortnite|GetFortPlayerControllerfromActor (Ability|GetAvatarActorfromActorInfo))) (CallFunction|GetTrackedProtectorBountyTarget))) (Utilities|Text|FormatText (.EventSubtitle (Utilities|Struct|BreakUrgentQuestData (Class|FortUrgentQuestItemDefinition|GetUrgentQuestData _asfort_urgent_quest_item_definition))) (Class|FortPlayerControllerGameplay|GetSettingsFilteredPlayerName (Utilities|Casting|CastToFortPlayerControllerGameplay (Fortnite|GetFortPlayerControllerfromActor (Ability|GetAvatarActorfromActorInfo))) (CallFunction|GetTrackedProtectorBountyTarget)))))
      (CallFunction|StartUrgentQuestEvent _structout))))

(event Custom|QuestCompleted
  (params (Title :type Text) (Subtitle :type Text))
  ;[0 0 0] Server_QuestCompleted
  (Variables|Default|SetIsQuestCompleted true)
  (CallFunction|DisplayMessageSuccess Title Subtitle)
  ;; Support Quest V2
  (CallFunction|BroadcastCustomVerb "(TagName=\"CustomVerb.Bounty.ProtectorTargetEliminated\")")
  (CallFunction|Cleanup)
  ;[0.2135 0.2135 0.2135] This is a time out in case the player does not respond with a ServerEndQuest
  (Utilities|FlowControl|Delay 2.0)
  (CallFunction|ServerEndQuest))

(event Custom|QuestFailed
  (params (Silent :type Boolean) (Title :type Text) (Subtitle :type Text))
  ;[0 0 0] Server_QuestFailed
  (Variables|Default|SetIsQuestCompleted false)
  (if (not Silent)
    (CallFunction|DisplayMessageFailure Title Subtitle))
  ;; Support Quest V2
  (CallFunction|BroadcastCustomVerb "(TagName=\"CustomVerb.Bounty.ProtectorTargetPoached\")")
  (CallFunction|Cleanup)
  ;;[0.2135 0.2135 0.2135] This is a time out in case the player does not respond with a ServerEndQuest
  (Utilities|FlowControl|Delay 2.0)
  (CallFunction|ServerEndQuest))

(event Custom|ServerEndQuest
  ;[0 0 0] Server End Quest with Success/Failure and Clean up
  (Utilities|FlowControl|DoOnce false
    (:Completed
      (if (Variables|Default|GetIsQuestCompleted)
        (CallFunction|CompleteQuest)
        (else
          (CallFunction|FailQuest))))))

(event Custom|DisplayMessage_Failure
  (params (Title :type Text) (Subtitle :type Text))
  ;[0 0 0] DisplayMessage_Failure
  (Class|AthenaHUDContext|ShowUrgentQuestEndedMessage (BlueprintContext|GetAthenaHUDContext) (Utilities|Struct|MakeUrgentQuestData :EventTag "(TagName=\"SpecialEvent.HuntingPlayer.Protector\")" :VisualStyleTag "(GameplayTags=)" :TotalEventTime 6e+01 :bDisplayTimer true :bQuestCompleteOnTimeExpiration false :bQuickIntro false :bQuickOutro false :bRewardPersist false :bDisplayMultiStepExpanded true :EventTitle Title :EventSubtitle Subtitle :EventDescriptionOverrideTag "(TagName=\"\")" :SortPriority 0 :bHideQuestProgressBar false :bDisplayAllObjectivesIndividually true :QuestProgressTextFormat "LOCGEN_FORMAT_ORDERED(INVTEXT(\"`{{0}} / `{{1}}\"), INVTEXT(\"CurrentProgress\"), INVTEXT(\"MaxProgress\"))" :bReverseProgressBar true :bShowBountyThreatInformation false :bShowBountyPriceAndDistance false :bDisplayStartedEventUI true :bIsSharedUrgentQuest false :SpecialEventStartedSound "/Game/Sounds/Fort_UI_Sounds/Gameplay/Bounty/Bounty_FailBounty_Cue.Bounty_FailBounty_Cue" :AlertIconBorderColor "(R=1.000000,G=0.456411,B=0.000000,A=1.000000)" :AlertIconBrushSize "(X=0.000000,Y=0.000000)" :bOverrideWidgetPrimaryColor false :OverriddenWidgetPrimaryColor "(R=1.000000,G=0.456411,B=0.000000,A=1.000000)" :ObjectiveCategory "Default") false)
  (CallFunction|ServerEndQuest))

(event Custom|DisplayMessage_Success
  (params (Title :type Text) (Subtitle :type Text))
  ;[0 0 0] DisplayMessage_Success
  (Class|AthenaHUDContext|ShowUrgentQuestEndedMessage (BlueprintContext|GetAthenaHUDContext) (Utilities|Struct|MakeUrgentQuestData :EventTag "(TagName=\"SpecialEvent.HuntingPlayer.Protector\")" :VisualStyleTag "(GameplayTags=)" :TotalEventTime 6e+01 :bDisplayTimer true :bQuestCompleteOnTimeExpiration false :bQuickIntro false :bQuickOutro false :bRewardPersist false :bDisplayMultiStepExpanded true :EventTitle Title :EventSubtitle Subtitle :EventDescriptionOverrideTag "(TagName=\"\")" :SortPriority 0 :bHideQuestProgressBar false :bDisplayAllObjectivesIndividually true :QuestProgressTextFormat "LOCGEN_FORMAT_ORDERED(INVTEXT(\"`{{0}} / `{{1}}\"), INVTEXT(\"CurrentProgress\"), INVTEXT(\"MaxProgress\"))" :bReverseProgressBar true :bShowBountyThreatInformation false :bShowBountyPriceAndDistance false :bDisplayStartedEventUI true :bIsSharedUrgentQuest false :SpecialEventStartedSound "/Game/Sounds/Fort_UI_Sounds/Gameplay/Bounty/Bounty_Complete_Cue.Bounty_Complete_Cue" :AlertIconBorderColor "(R=1.000000,G=0.456411,B=0.000000,A=1.000000)" :AlertIconBrushSize "(X=0.000000,Y=0.000000)" :bOverrideWidgetPrimaryColor false :OverriddenWidgetPrimaryColor "(R=1.000000,G=0.456411,B=0.000000,A=1.000000)" :ObjectiveCategory "Default") true)
  (CallFunction|ServerEndQuest))

(event Custom|OnProtectionTargetDied_Bind
  (params (PawnDamageInfo :type Pawn%20Damage%20Data%20Structure%20%28by%20ref%29))
  ;[0 0 0] OnProtectionTargetDied_Bind
  (if (Class|AthenaFunctionLibrary|RowtoBool (Variables|Default|GetAllowPoaching))
    (CallFunction|QuestFailed false "NSLOCTEXT(\"[A31A63EF45F22A15B2D0C599C177E596]\", \"23F2A55143521BF952C72EAA6A320E6A\", \"Bounty Ended\")" "NSLOCTEXT(\"[A31A63EF45F22A15B2D0C599C177E596]\", \"51F5C09C44D9F8B2212B6B98F6B758E5\", \"Squadmate Eliminated\")")))

(event Custom|OnHuntTargetDied_Bind
  (params (PawnDamageInfo :type Pawn%20Damage%20Data%20Structure%20%28by%20ref%29))
  (bind (_damagedactor _damage _instigatedby) (Utilities|Struct|BreakPawnDamageData PawnDamageInfo))
  (bind _returnvalue (Ability|GetAvatarActorfromActorInfo))
  (bind (_asfort_player_pawn_athena _bsuccess) (Utilities|Casting|CastToFortPlayerPawnAthena _returnvalue))
  (bind (_asplayer_pawn_athena _bsuccess_2) (Utilities|Casting|CastToPlayerPawn_Athena _damagedactor))
  ;[0 0 0] OnHuntTargetDied_Bind
  (if (and _bsuccess_2 _bsuccess)
    (Variables|Default|SetVictimPawn _asplayer_pawn_athena)
    (bind _output_get (Variables|Default|SetKillerPawn (Utilities|Casting|CastToFortPlayerPawnAthena (Pawn|GetControlledPawn _instigatedby))))
    (bind _output_get_2 (Variables|Default|SetOwningPawn _asfort_player_pawn_athena))
    (if (or (Squad|OnSameSquad _output_get _output_get_2) (and (Team|OnSameTeam _output_get _output_get_2) (GameplayTags|HasMatchingGameplayTag _output_get "(TagName=\"Athena.Gameplay.Status.Converted\")")))
      (CallFunction|QuestCompleted "NSLOCTEXT(\"[A31A63EF45F22A15B2D0C599C177E596]\", \"C84B4A234C75D5E3090B6D8E5E3FC083\", \"Bounty Completed\")" "NSLOCTEXT(\"[A31A63EF45F22A15B2D0C599C177E596]\", \"731446164AD7B52D75171A857A5DE1A4\", \"Squadmate Protected\")")
      (elif (Class|AthenaFunctionLibrary|RowtoBool (Variables|Default|GetAllowPoaching))
        (CallFunction|QuestFailed false "NSLOCTEXT(\"[A31A63EF45F22A15B2D0C599C177E596]\", \"23F2A55143521BF952C72EAA6A320E6A\", \"Bounty Ended\")" "NSLOCTEXT(\"[A31A63EF45F22A15B2D0C599C177E596]\", \"51F5C09C44D9F8B2212B6B98F6B758E5\", \"Squadmate Eliminated\")")))))

(event Custom|OneMinuteLeft
  ;[0 0 0] One Minute Left Warning
  (Audio|PlaySound2D "/Game/Sounds/Fort_UI_Sounds/Gameplay/Bounty/Bounty_TimeUp_Cue.Bounty_TimeUp_Cue"))

(event Custom|Cleanup
  ;[0 0 0] Cleanup
  (Class|FortControllerComponentTransientQuests|StopPlayerBountyThreatLevelUpdates (Class|FortPlayerPawn|GetTransientQuestsComponent (CallFunction|GetAvatarasFortPlayerPawn)))
  (Utilities|Time|ClearandInvalidateTimerbyHandle (Variables|Default|GetOneMinuteLeftTimerHandle))
  (CallFunction|CleanupBountyProtector))

(event Ability|EventOnEndAbility
  (params (bWasCancelled :type Boolean))
  ;[0 0 0] OnEndAbility
  (CallFunction|Cleanup))

(event Custom|BroadcastCustomVerb
  (params (CustomVerb :type Gameplay%20Tag%20Structure))
  (bind _returnvalue (Ability|GetAvatarActorfromActorInfo))
  (bind _returnvalue_2 (Fortnite|GetFortPlayerControllerfromActor _returnvalue))
  (bind _playerstate (Variables|/Script/Engine.Controller|GetPlayerState _returnvalue_2))
  (Utilities|IsValid _returnvalue_2
    (:Is%20Valid
      (Utilities|IsValid _playerstate
        (:Is%20Valid
          (CustomVerb|BroadcastCustomVerbMessage (CustomVerb|MakeCustomVerbMessage _playerstate CustomVerb)))))))

(event Custom|QuestStarted)

(fn GetBountyTarget
  (return (CallFunction|GetTrackedHunterBountyTarget)))
